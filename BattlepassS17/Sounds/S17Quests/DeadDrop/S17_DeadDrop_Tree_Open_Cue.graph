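# Sound cue for the Season 17 battle pass Dead Drop quest, played when a dead drop tree opens.
# It plays a single tree-open sound wave directly through one wave player, with no randomization or layering.
# It is routed to the quest event callout sound class and uses the shared Season 17 quest attenuation for spatial falloff.
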SoundCue /BattlepassS17/Sounds/S17Quests/DeadDrop/S17_DeadDrop_Tree_Open_Cue
root=WavePlayer_0
nodes [{id=WavePlayer_0, properties {soundWaveAssetPtr=&/S17_DeadDrop_Tree_Open}}]
properties {
 attenuationSettings=/BattlepassS17/Sounds/S17_Quest_Att
 maxDistance=4000
 soundClassObject=/Game/Sounds/FortSoundClassesAndModes/Fort_QuestEvent_Callout
}
